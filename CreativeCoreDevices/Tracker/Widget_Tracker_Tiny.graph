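# Compact HUD widget for the Creative Tracker device, showing a quest icon and current/target count on a tinted border.
# Tracker view model handlers recolor the divider, target text and border's dynamic material, and store the show-completion flag.
# Percentage changes, checked against a cached value, pulse the count below 100% or, if flagged, play a checkmark outro that fades the widget.
# The CheckMarkCompleted animation is not played by these graphs.
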WidgetBlueprint /CreativeCoreDevices/Tracker/Widget_Tracker_Tiny
animations [
 OnCurrentIncrement {
  binding_count=4
  length_seconds=0.5008
  object=&self:OnCurrentIncrement
  sequence {
   bindings [
    CanvasPanel_0 {
     id=#1
     possessed_class=CanvasPanel
     tracks [{
       type=MovieSceneFloatTrack
       display_name=Render Opacity
       sections [{type=MovieSceneFloatSection, end_frame=30, start_frame=0, channels [None {type=MovieSceneScriptingFloatChannel, const [1,3]}]}]
      }]
    }
    HBox_Progression {
     id=#2
     possessed_class=HorizontalBox
     tracks [{
       type=MovieScene2DTransformTrack
       sections [{
         type=MovieScene2DTransformSection
         end_frame=30
         start_frame=0
         channels [
          Translation.X {type=MovieSceneScriptingFloatChannel}
          Translation.Y {type=MovieSceneScriptingFloatChannel}
          Angle {type=MovieSceneScriptingFloatChannel}
          Scale.X {type=MovieSceneScriptingFloatChannel, keys [[0,1],[6,1.05],[30,1]]}
          Scale.Y {type=MovieSceneScriptingFloatChannel, keys [[0,1],[6,1.05],[30,1]]}
          Shear.X {type=MovieSceneScriptingFloatChannel}
          Shear.Y {type=MovieSceneScriptingFloatChannel}
         ]
        }]
      }]
    }
    Image_textglow {
     id=#3
     possessed_class=Image
     tracks [
      {
       type=MovieSceneFloatTrack
       display_name=Render Opacity
       sections [{type=MovieSceneFloatSection, channels [None {type=MovieSceneScriptingFloatChannel, keys [[0,0],[6,0.4],[30.046,0]]}]}]
      }
      {
       type=MovieScene2DTransformTrack
       sections [{
         type=MovieScene2DTransformSection
         channels [
          Translation.X {type=MovieSceneScriptingFloatChannel}
          Translation.Y {type=MovieSceneScriptingFloatChannel}
          Angle {type=MovieSceneScriptingFloatChannel}
          Scale.X {type=MovieSceneScriptingFloatChannel, keys [[0,1],[30.046,4]]}
          Scale.Y {type=MovieSceneScriptingFloatChannel, keys [[0,1],[30.046,4]]}
          Shear.X {type=MovieSceneScriptingFloatChannel}
          Shear.Y {type=MovieSceneScriptingFloatChannel}
         ]
        }]
      }
     ]
    }
    Image_textglow_2 {
     id=#4
     possessed_class=Image
     tracks [
      {
       type=MovieSceneFloatTrack
       display_name=Render Opacity
       sections [{
         type=MovieSceneFloatSection
         end_frame=30
         start_frame=0
         channels [None {type=MovieSceneScriptingFloatChannel, keys [[0,0],[6,0.1],[30.046,0]]}]
        }]
      }
      {
       type=MovieScene2DTransformTrack
       sections [{
         type=MovieScene2DTransformSection
         end_frame=30
         start_frame=0
         channels [
          Angle {type=MovieSceneScriptingFloatChannel, keys [[0,-2],[30,7]]}
          Scale.X {type=MovieSceneScriptingFloatChannel, keys [[0,0.14],[30,1]]}
          Scale.Y {type=MovieSceneScriptingFloatChannel, keys [[0,0.5],[30,1]]}
         ]
        }]
      }
     ]
    }
   ]
   display_rate {numerator=60}
  }
 }
 CheckMarkCompleted {
  binding_count=3
  length_seconds=0.6
  object=&self:CheckMarkCompleted
  sequence {bindings [
    Image_Glow {
     id=#5
     possessed_class=Image
     tracks [
      {
       type=MovieSceneFloatTrack
       display_name=Render Opacity
       sections [{type=MovieSceneFloatSection, channels [None {type=MovieSceneScriptingFloatChannel, keys [[6,0],[8,1],[12,0]]}]}]
      }
      {
       type=MovieScene2DTransformTrack
       sections [{
         type=MovieScene2DTransformSection
         channels [
          Translation.X {type=MovieSceneScriptingFloatChannel}
          Translation.Y {type=MovieSceneScriptingFloatChannel}
          Angle {type=MovieSceneScriptingFloatChannel}
          Scale.X {type=MovieSceneScriptingFloatChannel, keys [[6,1],[12,1.3]]}
          Scale.Y {type=MovieSceneScriptingFloatChannel, keys [[6,1],[12,1.3]]}
          Shear.X {type=MovieSceneScriptingFloatChannel}
          Shear.Y {type=MovieSceneScriptingFloatChannel}
         ]
        }]
      }
     ]
    }
    Image_Spark {
     id=#6
     possessed_class=Image
     tracks [
      {
       type=MovieSceneFloatTrack
       display_name=Render Opacity
       sections [{type=MovieSceneFloatSection, channels [None {type=MovieSceneScriptingFloatChannel, keys [[6,0],[8,1],[12,0]]}]}]
      }
      {
       type=MovieScene2DTransformTrack
       sections [{
         type=MovieScene2DTransformSection
         channels [
          Translation.X {type=MovieSceneScriptingFloatChannel}
          Translation.Y {type=MovieSceneScriptingFloatChannel}
          Angle {type=MovieSceneScriptingFloatChannel}
          Scale.X {type=MovieSceneScriptingFloatChannel, keys [[6,0.75],[12,1]]}
          Scale.Y {type=MovieSceneScriptingFloatChannel, keys [[6,0.75],[12,1]]}
          Shear.X {type=MovieSceneScriptingFloatChannel}
          Shear.Y {type=MovieSceneScriptingFloatChannel}
         ]
        }]
      }
     ]
    }
    CheckMarkIcon {
     id=#7
     possessed_class=Image
     tracks [
      {
       type=MovieSceneFloatTrack
       display_name=Render Opacity
       sections [{type=MovieSceneFloatSection, channels [None {type=MovieSceneScriptingFloatChannel, keys [[0,0],[7.2397,1]]}]}]
      }
      {
       type=MovieScene2DTransformTrack
       sections [{
         type=MovieScene2DTransformSection
         channels [
          Translation.X {type=MovieSceneScriptingFloatChannel}
          Translation.Y {type=MovieSceneScriptingFloatChannel, const [0,3]}
          Angle {type=MovieSceneScriptingFloatChannel}
          Scale.X {type=MovieSceneScriptingFloatChannel, keys [[0,2.3531],[6,0.45],[8,0.835]]}
          Scale.Y {type=MovieSceneScriptingFloatChannel, keys [[0,2.3531],[6,0.45],[8,0.835]]}
          Shear.X {type=MovieSceneScriptingFloatChannel}
          Shear.Y {type=MovieSceneScriptingFloatChannel}
         ]
        }]
      }
     ]
    }
   ]}
 }
 CheckMarkOutro {
  binding_count=5
  length_seconds=1.1
  object=&self:CheckMarkOutro
  sequence {bindings [
    CheckMarkIcon {
     id=#8
     possessed_class=Image
     tracks [
      {
       type=MovieSceneFloatTrack
       display_name=Render Opacity
       sections [{type=MovieSceneFloatSection, channels [None {type=MovieSceneScriptingFloatChannel, keys [[0,0],[7.2397,1]]}]}]
      }
      {
       type=MovieScene2DTransformTrack
       sections [{
         type=MovieScene2DTransformSection
         channels [
          Translation.X {type=MovieSceneScriptingFloatChannel}
          Translation.Y {type=MovieSceneScriptingFloatChannel, const [0,3]}
          Angle {type=MovieSceneScriptingFloatChannel}
          Scale.X {type=MovieSceneScriptingFloatChannel, keys [[0,2.3531],[6,0.45],[8,0.835]]}
          Scale.Y {type=MovieSceneScriptingFloatChannel, keys [[0,2.3531],[6,0.45],[8,0.835]]}
          Shear.X {type=MovieSceneScriptingFloatChannel}
          Shear.Y {type=MovieSceneScriptingFloatChannel}
         ]
        }]
      }
     ]
    }
    Image_Glow {
     id=#9
     possessed_class=Image
     tracks [
      {
       type=MovieSceneFloatTrack
       display_name=Render Opacity
       sections [{type=MovieSceneFloatSection, channels [None {type=MovieSceneScriptingFloatChannel, keys [[6,0],[8,1],[12,0]]}]}]
      }
      {
       type=MovieScene2DTransformTrack
       sections [{
         type=MovieScene2DTransformSection
         channels [
          Translation.X {type=MovieSceneScriptingFloatChannel}
          Translation.Y {type=MovieSceneScriptingFloatChannel}
          Angle {type=MovieSceneScriptingFloatChannel}
          Scale.X {type=MovieSceneScriptingFloatChannel, keys [[6,1],[12,1.3]]}
          Scale.Y {type=MovieSceneScriptingFloatChannel, keys [[6,1],[12,1.3]]}
          Shear.X {type=MovieSceneScriptingFloatChannel}
          Shear.Y {type=MovieSceneScriptingFloatChannel}
         ]
        }]
      }
     ]
    }
    CanvasPanel_0 {
     id=#10
     possessed_class=CanvasPanel
     tracks [{
       type=MovieSceneFloatTrack
       display_name=Render Opacity
       sections [{type=MovieSceneFloatSection, channels [None {type=MovieSceneScriptingFloatChannel, keys [[17,1],[22,0]]}]}]
      }]
    }
    HBox_Progression {
     id=#11
     possessed_class=HorizontalBox
     tracks [{
       type=MovieSceneFloatTrack
       display_name=Render Opacity
       sections [{type=MovieSceneFloatSection, end_frame=3, start_frame=0, channels [None {type=MovieSceneScriptingFloatChannel, keys [[0,1],[3,0]]}]}]
      }]
    }
    Image_Spark {
     id=#12
     possessed_class=Image
     tracks [
      {
       type=MovieSceneFloatTrack
       display_name=Render Opacity
       sections [{type=MovieSceneFloatSection, channels [None {type=MovieSceneScriptingFloatChannel, keys [[6,0],[8,1],[12,0]]}]}]
      }
      {
       type=MovieScene2DTransformTrack
       sections [{
         type=MovieScene2DTransformSection
         channels [
          Translation.X {type=MovieSceneScriptingFloatChannel}
          Translation.Y {type=MovieSceneScriptingFloatChannel}
          Angle {type=MovieSceneScriptingFloatChannel}
          Scale.X {type=MovieSceneScriptingFloatChannel, keys [[6,0.75],[12,1]]}
          Scale.Y {type=MovieSceneScriptingFloatChannel, keys [[6,0.75],[12,1]]}
          Shear.X {type=MovieSceneScriptingFloatChannel}
          Shear.Y {type=MovieSceneScriptingFloatChannel}
         ]
        }]
      }
     ]
    }
   ]}
 }
]
graphs [
 EventGraph {links [], nodes []}
 MVVM Color Changed {nodes [
   n1 = MVVM Color Changed!() -> (then, Color:Linear Color Structure) {is=title K2Node_FunctionEntry}
   > Branch(Condition=Get IsViewModelInitialized!(self:Device - Tracker View Model Object Reference=Get MVVM_UEFN_Tracker() -> (MVVM_UEFN_Tracker:Device - Tracker View Model Object Reference)) -> (IsViewModelInitialized:bool))
   > SetColorAndOpacity(
    self=[
     Get Text_TargetAmount() -> (Text_TargetAmount:Common Text Object Reference)
     Get Text_ProgressDivider() -> (Text_ProgressDivider:Common Text Object Reference)
    ]
    InColorAndOpacity=Make Slate Color(SpecifiedColor=@n1)
   )
   > n2 = GetDynamicMaterial(self:Border Object Reference=Get QuestBackgroundBorder() -> (QuestBackgroundBorder:Common Border Object Reference))
   > SetVectorParameterValue(self=@n2, ParameterName=Border_Color, Value=@n1)
  ]}
 MVVM Percentage Changed {nodes [
   n1 = MVVM Percentage Changed!() -> (then, Percentage:double=~) {is=title K2Node_FunctionEntry}
   n2 = Get CachedPercentage() -> (CachedPercentage:double)
   n3 = Get MVVM_UEFN_Tracker()
   > Branch(Condition=AND Boolean(A="Not Equal (Float)"(A=@n1, B=@n2), B=Get IsViewModelInitialized(self=@n3)))
   > n4 = Branch(Condition="float >= float"(A=@n1, B=1))
   > Branch(Condition=Get ShowCompleteAnimation() -> (ShowCompleteAnimation:bool))
   > n5 = PlayAnimation(InAnimation=Get CheckMarkOutro() -> (CheckMarkOutro:Widget Animation Object Reference))
   Set CachedPercentage(execute=[@n5.then, @n6.then], CachedPercentage:double=Get Percentage() -> (Percentage:double)) -> (Output_Get:double)
   n6 = PlayAnimation(execute=@n4.else, InAnimation=Get OnCurrentIncrement() -> (OnCurrentIncrement:Widget Animation Object Reference))
  ]}
 MVVM Show Complete Changed {nodes [
   n1 = MVVM Show Complete Changed!() -> (then, ShowCompleteAnimation:bool=~) {is=title K2Node_FunctionEntry}
   > Set ShowCompleteAnimation(ShowCompleteAnimation:bool=@n1) -> (Output_Get:bool)
  ]}
]
variables [
 CachedPercentage {type=number}
 ShowCompleteAnimation {type=boolean}
]
widgets {
 info {inheritedWidgetCount=0, parentClass=&UserWidget, rootWidgetClass=&CanvasPanel, widgetCount=25}
 widgets [
  CanvasPanel_0 {widgetClassPath=&CanvasPanel, widget=CanvasPanel_0}
  SizeBox_Root {
   widgetClassPath=&SizeBox
   bIsVariable=true
   parent=CanvasPanel_0
   widget=SizeBox_Root
   properties {bOverride_MinDesiredWidth=true, minDesiredWidth=70, widthOverride=220.6998}
   slot {class=CanvasPanelSlot, properties {bAutoSize=true, layoutData=~}}
  }
  VerticalBox_79 {
   widgetClassPath=&VerticalBox
   parent=SizeBox_Root
   widget=VerticalBox_79
   slot {class=SizeBoxSlot, properties {verticalAlignment=VAlign_Center}}
  }
  Overlay_NewTracker {
   widgetClassPath=&Overlay
   parent=VerticalBox_79
   widget=Overlay_NewTracker
   slot {class=VerticalBoxSlot, properties {size {sizeRule=Fill, value=1}}}
  }
  QuestBackgroundBorder {
   widgetClassPath=&CommonBorder
   bIsVariable=true
   parent=Overlay_NewTracker
   widget=QuestBackgroundBorder
   properties {
    background {
     bIsDynamicallyLoaded=false
     drawAs=Box
     imageType=NoImage
     mirroring=NoMirror
     resourceName=None
     resourceObject=/FortniteCoreUI/Athena/UI/InGame/Creative/Art/M_UI_Quest_Container_QuestTracker_Inst
     tiling=NoTile
     imageSize=<64 64>
     margin=<ltrb 0.5 0.5 0.5 0.5>
     outlineSettings {
      bUseBrushTransparency=false
      roundingType=HalfHeightRadius
      width=0
      color {colorUseRule=UseColor_Specified, specifiedColor=<rgba 0 0 0 0>}
      cornerRadii=<0 0 0 0>
     }
     tintColor {colorUseRule=UseColor_Specified, specifiedColor=<rgba 1 1 1 1>}
     uVRegion {bIsValid=false, max=<0 0>, min=<0 0>}
    }
    padding=<ltrb 0 0 0 0>
   }
   slot {class=OverlaySlot, properties {horizontalAlignment=HAlign_Fill, verticalAlignment=VAlign_Fill}}
  }
  HorizontalBox_1 {
   widgetClassPath=&HorizontalBox
   parent=Overlay_NewTracker
   widget=HorizontalBox_1
   slot {class=OverlaySlot, properties {horizontalAlignment=HAlign_Fill, verticalAlignment=VAlign_Fill}}
  }
  SizeBox_0 {
   widgetClassPath=&SizeBox
   parent=HorizontalBox_1
   widget=SizeBox_0
   properties {heightOverride=48, minDesiredHeight=48, minDesiredWidth=24, widthOverride=48}
   slot {class=HorizontalBoxSlot, properties {horizontalAlignment=HAlign_Left, padding=<ltrb 25 5 5 5>, size {sizeRule=Automatic, value=0.01}}}
  }
  Overlay_1 {widgetClassPath=&Overlay, parent=SizeBox_0, widget=Overlay_1, slot {class=SizeBoxSlot}}
  QuestIcon {
   widgetClassPath=&Image
   bIsVariable=true
   parent=Overlay_1
   widget=QuestIcon
   properties {brush {
     bIsDynamicallyLoaded=false
     drawAs=Image
     imageType=NoImage
     mirroring=NoMirror
     resourceName=None
     resourceObject=/Game/Athena/UI/InGame/Creative/PropertyEditor/Icon_Picker/Textures/128x128/T_UI_IconLibrary_information_128
     tiling=NoTile
     imageSize=<48 48>
     margin=<ltrb 0 0 0 0>
     outlineSettings {
      bUseBrushTransparency=false
      roundingType=HalfHeightRadius
      width=0
      color {colorUseRule=UseColor_Specified, specifiedColor=<rgba 0 0 0 0>}
      cornerRadii=<0 0 0 0>
     }
     tintColor {colorUseRule=UseColor_Specified, specifiedColor=<rgba 1 1 1 1>}
     uVRegion {bIsValid=false, max=<0 0>, min=<0 0>}
    }}
   slot {class=OverlaySlot}
  }
  Spacer_63 {widgetClassPath=&Spacer, parent=Overlay_1, widget=Spacer_63, properties {size=<5 48>}, slot {class=OverlaySlot}}
  Spacer_102 {
   widgetClassPath=&Spacer
   parent=HorizontalBox_1
   widget=Spacer_102
   slot {class=HorizontalBoxSlot, properties {size {sizeRule=Fill, value=1}}}
  }
  BoxProgression {
   widgetClassPath=&HorizontalBox
   bIsVariable=true
   parent=HorizontalBox_1
   widget=BoxProgression
   slot {class=HorizontalBoxSlot, properties {horizontalAlignment=HAlign_Right, verticalAlignment=VAlign_Center}}
  }
  HorizontalBox_4 {
   widgetClassPath=&HorizontalBox
   parent=BoxProgression
   widget=HorizontalBox_4
   slot {class=HorizontalBoxSlot, properties {horizontalAlignment=HAlign_Right}}
  }
  Progress {
   widgetClassPath=&Overlay
   bIsVariable=true
   parent=HorizontalBox_4
   widget=Progress
   slot {class=HorizontalBoxSlot, properties {horizontalAlignment=HAlign_Right}}
  }
  HBox_Progression {
   widgetClassPath=&HorizontalBox
   parent=Progress
   widget=HBox_Progression
   slot {class=OverlaySlot, properties {horizontalAlignment=HAlign_Right, padding=<ltrb 0 0 10 0>}}
  }
  Text_CurrentAmount {
   widgetClassPath=&CommonTextBlock
   bIsVariable=true
   parent=HBox_Progression
   widget=Text_CurrentAmount
   properties {
    text=1
    colorAndOpacity {colorUseRule=UseColor_Specified, specifiedColor=<rgba 0.6038 0.8879 1 1>}
    font {
     bForceMonospaced=false
     bMaterialIsStencil=false
     fontMaterial=null
     fontObject=&BurbankSmall
     letterSpacing=0
     monospacedWidth=1
     size=15
     skewAmount=0
     typefaceFontName=Bold
     outlineSettings {
      bApplyOutlineToDropShadows=false
      bMiteredCorners=false
      bSeparateFillAlpha=false
      outlineMaterial=null
      outlineSize=0
      outlineColor=<rgba 0 0 0 1>
     }
    }
    shadowOffset=<0 0>
   }
   slot {class=HorizontalBoxSlot}
  }
  Text_ProgressDivider {
   widgetClassPath=&CommonTextBlock
   bIsVariable=true
   parent=HBox_Progression
   widget=Text_ProgressDivider
   properties {
    text=/
    colorAndOpacity {colorUseRule=UseColor_Specified, specifiedColor=<rgba 0.2195 0.6584 1 1>}
    font {
     bForceMonospaced=false
     bMaterialIsStencil=false
     fontMaterial=null
     fontObject=&BurbankSmall
     letterSpacing=0
     monospacedWidth=1
     size=15
     skewAmount=0
     typefaceFontName=Bold
     outlineSettings {
      bApplyOutlineToDropShadows=false
      bMiteredCorners=false
      bSeparateFillAlpha=false
      outlineMaterial=null
      outlineSize=0
      outlineColor=<rgba 0 0 0 1>
     }
    }
    shadowOffset=<0 0>
   }
   slot {class=HorizontalBoxSlot}
  }
  Text_TargetAmount {
   widgetClassPath=&CommonTextBlock
   bIsVariable=true
   parent=HBox_Progression
   widget=Text_TargetAmount
   properties {
    text=1
    colorAndOpacity {colorUseRule=UseColor_Specified, specifiedColor=<rgba 0.2195 0.6584 1 1>}
    font {
     bForceMonospaced=false
     bMaterialIsStencil=false
     fontMaterial=null
     fontObject=&BurbankSmall
     letterSpacing=0
     monospacedWidth=1
     size=15
     skewAmount=0
     typefaceFontName=Bold
     outlineSettings {
      bApplyOutlineToDropShadows=false
      bMiteredCorners=false
      bSeparateFillAlpha=false
      outlineMaterial=null
      outlineSize=0
      outlineColor=<rgba 0 0 0 1>
     }
    }
    shadowOffset=<0 0>
   }
   slot {class=HorizontalBoxSlot}
  }
  Image_textglow {
   widgetClassPath=&Image
   bIsVariable=true
   parent=Progress
   widget=Image_textglow
   properties {
    visibility=HitTestInvisible
    brush {
     bIsDynamicallyLoaded=false
     drawAs=Image
     imageType=NoImage
     mirroring=NoMirror
     resourceName=None
     resourceObject=/Game/Athena/HUD/LTM/Wax/Materials/M_UI_TextGlow_Add
     tiling=NoTile
     imageSize=<16 16>
     margin=<ltrb 0 0 0 0>
     outlineSettings {
      bUseBrushTransparency=false
      roundingType=HalfHeightRadius
      width=0
      color {colorUseRule=UseColor_Specified, specifiedColor=<rgba 0 0 0 0>}
      cornerRadii=<0 0 0 0>
     }
     tintColor {colorUseRule=UseColor_Specified, specifiedColor=<rgba 0.25 0.25 0.25 1>}
     uVRegion {bIsValid=false, max=<0 0>, min=<0 0>}
    }
    colorAndOpacity=<rgba 0 0.3465 0.9911 1>
    renderTransform {angle=0, scale=<0 0>, shear=<0 0>, translation=<0 0>}
   }
   slot {class=OverlaySlot, properties {horizontalAlignment=HAlign_Center, verticalAlignment=VAlign_Center}}
  }
  Image_textglow_2 {
   widgetClassPath=&Image
   bIsVariable=true
   parent=Progress
   widget=Image_textglow_2
   properties {
    renderOpacity=0
    visibility=SelfHitTestInvisible
    brush {
     bIsDynamicallyLoaded=false
     drawAs=Image
     imageType=NoImage
     mirroring=NoMirror
     resourceName=None
     resourceObject=&M_UI_TextGlow_Add
     tiling=NoTile
     imageSize=<16 10>
     margin=<ltrb 0 0 0 0>
     outlineSettings {
      bUseBrushTransparency=false
      roundingType=HalfHeightRadius
      width=0
      color {colorUseRule=UseColor_Specified, specifiedColor=<rgba 0 0 0 0>}
      cornerRadii=<0 0 0 0>
     }
     tintColor {colorUseRule=UseColor_Specified, specifiedColor=<rgba 0.75 0.75 0.75 1>}
     uVRegion {bIsValid=false, max=<0 0>, min=<0 0>}
    }
    colorAndOpacity=<rgba 0 0.7053 0.9911 1>
    renderTransform {angle=7, scale=<1 1>, shear=<0 0>, translation=<0 0>}
   }
   slot {class=OverlaySlot, properties {horizontalAlignment=HAlign_Fill, verticalAlignment=VAlign_Fill, padding=<ltrb -40 2 -40 2>}}
  }
  Completion {
   widgetClassPath=&SizeBox
   bIsVariable=true
   parent=Progress
   widget=Completion
   properties {bOverride_HeightOverride=true, bOverride_WidthOverride=true, heightOverride=25, visibility=HitTestInvisible, widthOverride=25}
   slot {class=OverlaySlot, properties {horizontalAlignment=HAlign_Right, verticalAlignment=VAlign_Bottom, padding=<ltrb 0 0 10 0>}}
  }
  Overlay_0 {widgetClassPath=&Overlay, parent=Completion, widget=Overlay_0, slot {class=SizeBoxSlot}}
  CheckMarkIcon {
   widgetClassPath=&Image
   bIsVariable=true
   parent=Overlay_0
   widget=CheckMarkIcon
   properties {
    renderOpacity=0
    brush {
     bIsDynamicallyLoaded=false
     drawAs=Image
     imageType=NoImage
     mirroring=NoMirror
     resourceName=None
     resourceObject=/Game/UI/Foundation/Textures/Icons/Checks/T-Icon-Check-64
     tiling=NoTile
     imageSize=<64 64>
     margin=<ltrb 0 0 0 0>
     outlineSettings {
      bUseBrushTransparency=false
      roundingType=HalfHeightRadius
      width=0
      color {colorUseRule=UseColor_Specified, specifiedColor=<rgba 0 0 0 0>}
      cornerRadii=<0 0 0 0>
     }
     tintColor {colorUseRule=UseColor_Specified, specifiedColor=<rgba 0.0343 1 0.227 1>}
     uVRegion {bIsValid=false, max=<0 0>, min=<0 0>}
    }
   }
   slot {class=OverlaySlot, properties {horizontalAlignment=HAlign_Center, verticalAlignment=VAlign_Center}}
  }
  Image_Glow {
   widgetClassPath=&Image
   bIsVariable=true
   parent=Overlay_0
   widget=Image_Glow
   properties {
    renderOpacity=0
    brush {
     bIsDynamicallyLoaded=false
     drawAs=Image
     imageType=NoImage
     mirroring=NoMirror
     resourceName=None
     resourceObject=&M_UI_TextGlow_Add
     tiling=NoTile
     imageSize=<130 130>
     margin=<ltrb 0 0 0 0>
     outlineSettings {
      bUseBrushTransparency=false
      roundingType=HalfHeightRadius
      width=0
      color {colorUseRule=UseColor_Specified, specifiedColor=<rgba 0 0 0 0>}
      cornerRadii=<0 0 0 0>
     }
     tintColor {colorUseRule=UseColor_Specified, specifiedColor=<rgba 0 1 0.1443 1>}
     uVRegion {bIsValid=false, max=<0 0>, min=<0 0>}
    }
   }
   slot {class=OverlaySlot, properties {horizontalAlignment=HAlign_Center, verticalAlignment=VAlign_Center, padding=<ltrb -100 -100 -100 -100>}}
  }
  Image_Spark {
   widgetClassPath=&Image
   bIsVariable=true
   parent=Overlay_0
   widget=Image_Spark
   properties {
    renderOpacity=0
    brush {
     bIsDynamicallyLoaded=false
     drawAs=Image
     imageType=NoImage
     mirroring=NoMirror
     resourceName=None
     resourceObject=/Game/UI/InGame/Quest/Materials/M_UI_SparkStar
     tiling=NoTile
     imageSize=<120 120>
     margin=<ltrb 0 0 0 0>
     outlineSettings {
      bUseBrushTransparency=false
      roundingType=HalfHeightRadius
      width=0
      color {colorUseRule=UseColor_Specified, specifiedColor=<rgba 0 0 0 0>}
      cornerRadii=<0 0 0 0>
     }
     tintColor {colorUseRule=UseColor_Specified, specifiedColor=<rgba 1 1 1 1>}
     uVRegion {bIsValid=false, max=<0 0>, min=<0 0>}
    }
   }
   slot {class=OverlaySlot, properties {horizontalAlignment=HAlign_Center, verticalAlignment=VAlign_Center, padding=<ltrb -75 -75 -75 -75>}}
  }
 ]
}
